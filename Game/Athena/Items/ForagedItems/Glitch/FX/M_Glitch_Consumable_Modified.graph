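# Translucent unlit particle material for the Glitch foraged consumable's effect, drawing flickering digital-noise blocks in two blended colors.
# Time-sampled noise jitters a thresholded rectangular mask whose position and size come from dynamic parameters and particle red.
# Panned digital-noise atlas channels drive emissive color and break up the mask for opacity, with desaturation on mobile.
# Static switches allow full transparency or mask-driven refraction, which is shut down by default.
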Material /Game/Athena/Items/ForagedItems/Glitch/FX/M_Glitch_Consumable_Modified
graph {
 properties {
  bUsedWithNiagaraSprites=true
  bUsedWithParticleSprites=true
  blendMode=BLEND_Translucent
  pixelDepthOffsetMode=PDOM_Legacy
  refractionMethod=RM_IndexOfRefraction
  shadingModel=MSM_Unlit
  usedShadingModels=Unlit
 }
 nodes [
  n1 = Constant() {r=2}
  n2 = Add(A=Multiply(A=MaterialFunctionCall(Speed=@n1) {materialFunction=&DebugTimeSine}.Result) {constB=5})
  n3 = DynamicParameter() {paramNames [X_position,Y_position,Width0,Width1]}
  n4 = Multiply(A=@n3.Y_position) {constB=-1}
  n5 = Time()
  n6 = Multiply(A=@n5) {constB=2}
  n7 = TextureSample(UVs=AppendVector(A=@n5, B=Frac(None=@n6))) {texture=&T_NoisePack_01}
  n8 = OneMinus(None=@n7.G)
  n9 = Panner(Coordinate=TextureCoordinate()) {speedX=0.1, speedY=1}
  n10 = ComponentMask(None=VectorParameter() {parameterName="Base_texture(R)", defaultValue=<rgba 5 0.05 0 1>}.RGB) {g=true, r=true}
  n11 = TextureSample(UVs=Add(A=LinearInterpolate(A=@n4, B=@n3.Y_position, Alpha=@n8), B=Multiply(A=@n9, B=@n10))) {samplerType=SAMPLERTYPE_LinearColor, texture=/Game/Global/Textures/Default/Noise/T_Atlas_DigitalNoise_01}
  n12 = LinearInterpolate(
   A=VectorParameter() {parameterName=Color_1, defaultValue=<rgba 1 0.0938 0.501 1>}.RGB
   B=LinearInterpolate(B=VectorParameter() {parameterName=Color, defaultValue=<rgba 0.0938 0.6737 1 1>}.RGB, Alpha=LinearInterpolate(B=@n2, Alpha=@n11.R))
   Alpha=@n11.R
  )
  n13 = Desaturation(None=@n12, Fraction=ScalarParameter() {parameterName=Desaturation_for_Android})
  MaterialFunctionCall(Emissive=Multiply(A=ShadingPathSwitch(Default=@n12, Deferred=@n12, Forward=@n12, Mobile=@n13)) {constB=0.7}) {materialFunction=&MF_FX_EmissiveAdjustments}
  n14 = VectorParameter() {parameterName=Distort_map_ctrl, defaultValue=<rgba 3 0.02 0 0>}
  n15 = Panner(Speed=AppendVector(A=@n14.B, B=@n14.B))
  n16 = ComponentMask(None=@n14.RGB) {g=true, r=true}
  Multiply(A=TextureSample(UVs=Multiply(A=@n15, B=@n16)) {samplerType=SAMPLERTYPE_LinearColor, texture=&T_Atlas_DigitalNoise_01}.RGB) {constB=0.1}
  OneMinus(None=@n7.R)
  n17 = Multiply(
   A=Panner(Coordinate=TextureCoordinate()) {speedX=-0.2, speedY=0.2}
   B=ComponentMask(None=VectorParameter() {parameterName="Base_texture(B)", defaultValue=<rgba 4 0.02 0 1>}.RGB) {g=true, r=true}
  )
  n18 = DynamicParameter() {paramNames [X_position,Y_position,Width0,Width1]}
  n19 = Time()
  n20 = Multiply(A=@n19) {constB=5}
  n21 = TextureSample(UVs=AppendVector(A=@n19, B=Frac(None=@n20))) {texture=&T_NoisePack_01}
  n22 = LinearInterpolate(A=@n18.Width0, B=@n18.Width0, Alpha=OneMinus(None=@n21.B))
  n23 = TextureCoordinate()
  n24 = LinearInterpolate(A=Multiply(A=@n18.X_position) {constB=-1}, B=@n18.X_position, Alpha=OneMinus(None=@n21.R))
  n25 = LinearInterpolate(A=Add(B=@n22) {constB=0}, B=Subtract(B=@n22) {constB=0}, Alpha=ComponentMask(None=Add(A=@n24, B=@n23)) {r=true})
  n26 = Constant() {r=0.25}
  n27 = Constant() {r=1}
  n28 = Constant()
  n29 = ParticleColor()
  n30 = Time()
  n31 = Multiply(A=@n30) {constB=2}
  n32 = TextureSample(UVs=AppendVector(A=@n30, B=Frac(None=@n31))) {samplerType=SAMPLERTYPE_Normal, texture=/Game/Global/Textures/Default/Noise/64x64Noise}
  n33 = LinearInterpolate(A=@n29.R, B=@n29.R, Alpha=OneMinus(None=@n32.R))
  n34 = DynamicParameter() {paramNames [X_position,Y_position,Width0,Width1]}
  n35 = LinearInterpolate(A=Multiply(A=@n34.Y_position) {constB=-1}, B=@n34.Y_position, Alpha=OneMinus(None=@n32.G))
  n36 = LinearInterpolate(A=Add(B=@n33) {constB=0}, B=Subtract(B=@n33) {constB=0}, Alpha=ComponentMask(None=Add(A=@n23, B=@n35) {constB=0}) {g=true})
  n37 = Constant() {r=0.25}
  n38 = Constant() {r=1}
  n39 = Constant()
  n40 = Multiply(
   A=If(A=@n25, B=Constant() {r=0.75}, "A > B"=Constant(), "A < B"=If(A=@n25, B=@n26, "A > B"=@n27, "A < B"=@n28))
   B=If(A=@n36, B=Constant() {r=0.75}, "A > B"=Constant(), "A < B"=If(A=@n36, B=@n37, "A > B"=@n38, "A < B"=@n39))
  )
  n41 = Multiply(
   A=Multiply(
    A=@n11.B
    B=TextureSample(UVs=@n17) {samplerType=SAMPLERTYPE_LinearColor, texture=&T_Atlas_DigitalNoise_01}.G
   )
   B=@n40
  )
  n42 = ScalarParameter() {parameterName=Noise Texture}
  StaticSwitchParameter(
   True=Constant()
   False=Multiply(A=LinearInterpolate(A=@n41, B=@n40, Alpha=@n42), B=ScalarParameter() {defaultValue=1, parameterName=OverallOpacity})
  ) {parameterName="Full_transparency(CheckWhenUseRefraction)"}
  n43 = Constant() {r=1}
  n44 = ScalarParameter() {defaultValue=1.2, parameterName=Refraction_ratio}
  StaticSwitchParameter(
   True=Constant() {r=1}
   False=LinearInterpolate(
    A=Constant() {r=1}
    B=LinearInterpolate(A=@n43, B=@n44, Alpha=@n40)
    Alpha=DynamicParameter() {parameterIndex=1, paramNames [Ref,Param2,Param3,Param4]}.Ref
   )
  ) {defaultValue=true, parameterName=ShutDown_refraction}
 ]
}
