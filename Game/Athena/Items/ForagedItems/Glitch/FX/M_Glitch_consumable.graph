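# Translucent unlit particle sprite material for the Glitch foraged consumable, producing a flickering digital-noise effect.
# Panning digital-noise atlas samples blend pink, cyan and a time-sine pulse, tinted by particle color.
# Noise channels multiply with a rectangular mask, built from If thresholds and positioned and sized by dynamic parameters, to form opacity.
# Static switches offer zero opacity and mask-driven index-of-refraction distortion, with refraction shut off by default.
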Material /Game/Athena/Items/ForagedItems/Glitch/FX/M_Glitch_consumable.M_Glitch_Consumable
graph {
 properties {
  bUsedWithParticleSprites=true
  blendMode=BLEND_Translucent
  pixelDepthOffsetMode=PDOM_Legacy
  refractionMethod=RM_IndexOfRefraction
  shadingModel=MSM_Unlit
  usedShadingModels=Unlit
 }
 nodes [
  n1 = Constant3Vector() {constant=<rgba 1 0.0938 0.501 0>}
  n2 = Multiply(A=MaterialFunctionCall(Speed=Constant() {r=2}) {materialFunction=&DebugTimeSine}.Result) {constB=5}
  n3 = Panner(Coordinate=TextureCoordinate()) {speedX=0.1, speedY=1}
  n4 = ComponentMask(None=VectorParameter() {parameterName="Base_texture(R)", defaultValue=<rgba 5 0.05 0 1>}.RGB) {g=true, r=true}
  n5 = TextureSample(UVs=Add(A=DynamicParameter() {paramNames [X_position,Y_position,With,Hight]}.Y_position, B=Multiply(A=@n3, B=@n4))) {samplerType=SAMPLERTYPE_LinearColor, texture=/Game/Global/Textures/Default/Noise/T_Atlas_DigitalNoise_01}
  n6 = LinearInterpolate(B=Constant3Vector() {constant=<rgba 0.0938 0.6737 1 0>}, Alpha=LinearInterpolate(B=Add(A=@n2), Alpha=@n5.R))
  MaterialFunctionCall(Emissive=Multiply(A=LinearInterpolate(A=@n1, B=@n6, Alpha=@n5.R), B=ParticleColor().RGB)) {materialFunction=&MF_FX_EmissiveAdjustments}
  n7 = VectorParameter() {parameterName=Distort_map_ctrl, defaultValue=<rgba 3 0.02 0 0>}
  n8 = Panner(Speed=AppendVector(A=@n7.B, B=@n7.B))
  n9 = ComponentMask(None=@n7.RGB) {g=true, r=true}
  Multiply(A=TextureSample(UVs=Multiply(A=@n8, B=@n9)) {samplerType=SAMPLERTYPE_LinearColor, texture=&T_Atlas_DigitalNoise_01}.RGB) {constB=0.1}
  n10 = TextureCoordinate()
  n11 = VectorParameter() {parameterName="Base_texture(B)", defaultValue=<rgba 4 0.02 0 1>}
  n12 = TextureSample(UVs=Multiply(A=Panner(Coordinate=@n10) {speedX=-0.2, speedY=0.2}, B=ComponentMask(None=@n11.RGB) {g=true, r=true})) {samplerType=SAMPLERTYPE_LinearColor, texture=&T_Atlas_DigitalNoise_01}
  n13 = DynamicParameter() {paramNames [X_position,Y_position,With,Hight]}
  n14 = TextureCoordinate()
  n15 = LinearInterpolate(A=Add(B=@n13.With) {constB=0}, B=Subtract(B=@n13.With) {constB=0}, Alpha=ComponentMask(None=Add(A=@n13.X_position, B=@n14)) {r=true})
  n16 = Constant() {r=0.25}
  n17 = Constant() {r=1}
  n18 = Constant()
  n19 = DynamicParameter() {paramNames [X_position,Y_position,With,Hight]}
  n20 = LinearInterpolate(
   A=Add(B=@n19.Hight) {constB=0}
   B=Subtract(B=@n19.Hight) {constB=0}
   Alpha=ComponentMask(None=Add(A=@n14, B=@n19.Y_position) {constB=0}) {g=true}
  )
  n21 = Constant() {r=0.25}
  n22 = Constant() {r=1}
  n23 = Constant()
  n24 = Multiply(
   A=If(A=@n15, B=Constant() {r=0.75}, "A > B"=Constant(), "A < B"=If(A=@n15, B=@n16, "A > B"=@n17, "A < B"=@n18))
   B=If(A=@n20, B=Constant() {r=0.75}, "A > B"=Constant(), "A < B"=If(A=@n20, B=@n21, "A > B"=@n22, "A < B"=@n23))
  )
  StaticSwitchParameter(True=Constant(), False=Multiply(A=Multiply(A=@n5.B, B=@n12.G), B=@n24)) {parameterName="Full_transparency(CheckWhenUseRefraction)"}
  StaticSwitchParameter(
   True=Constant() {r=1}
   False=LinearInterpolate(A=Constant() {r=1}, B=ScalarParameter() {defaultValue=1.2, parameterName=Refraction_ratio}, Alpha=@n24)
  ) {defaultValue=true, parameterName=ShutDown_refraction}
 ]
}
